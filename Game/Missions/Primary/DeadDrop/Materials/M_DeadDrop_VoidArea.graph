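# Additive unlit material for the Dead Drop mission's void area, glowing red-orange from the channels of a landing-zone mask texture.
# Two bump-offset samples of the green channel form a scalable parallax icon pulsed by WarningTime, added to a slower-pulsing blue channel and a tiled red pattern.
# Nested camera-depth fades reduce this to a dimmer red pattern, then black, with distance, and a depth fade softens intersections.
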Material /Game/Missions/Primary/DeadDrop/Materials/M_DeadDrop_VoidArea
graph {
 properties {blendMode=BLEND_Additive, pixelDepthOffsetMode=PDOM_Legacy, shadingModel=MSM_Unlit, usedShadingModels=Unlit}
 nodes [
  n1 = MaterialFunctionCall!(
   Sine Phase=Multiply(A=Time(), B=ScalarParameter() {defaultValue=0.2, parameterName=WarningTime})
   Value 1=Constant() {r=0.7}
   Value 2=Constant() {r=2}
  ) {materialFunction=/Engine/Functions/Engine_MaterialFunctions02/Utility/Sine_Remapped}
  n2 = Multiply(
   A=TextureCoordinate()
   B=AppendVector(A=ScalarParameter() {defaultValue=1, parameterName=IconScaleX}, B=ScalarParameter() {defaultValue=1, parameterName=IconScaleY})
  )
  n3 = BumpOffset(Coordinate=@n2, Height=Constant() {r=1})
  n4 = Constant() {r=2}
  n5 = Add(
   A=Multiply(A=TextureSample(UVs=@n3) {samplerType=SAMPLERTYPE_Masks, texture=/Game/Missions/Primary/DeadDrop/Textures/T_DeadDrop_LandingZone_M}.G) {constB=0.2}
   B=TextureSample(UVs=BumpOffset(Coordinate=@n2, Height=@n4)) {samplerType=SAMPLERTYPE_Masks, texture=&T_DeadDrop_LandingZone_M}.G
  )
  n6 = TextureSample() {samplerType=SAMPLERTYPE_Masks, texture=&T_DeadDrop_LandingZone_M}
  n7 = MaterialFunctionCall(Sine Phase=Multiply(A=Time(), B=Constant() {r=0.5}), Value 1=Constant() {r=0.2}, Value 2=Constant() {r=1}) {materialFunction=&Sine_Remapped}
  n8 = Add(
   A=LinearInterpolate(
    A=Constant3Vector() {constant=<rgba 0.035 0.0047 0 0>}
    B=Constant3Vector() {constant=<rgba 2 0.2727 0 0>}
    Alpha=Multiply(A=@n1.Remapped Sine, B=@n5)
   )
   B=Multiply(A=Multiply(A=@n6.B, B=@n7.Remapped Sine), B=Constant3Vector() {constant=<rgba 2 0.1402 0 0>})
  )
  n9 = TextureCoordinate()
  n10 = AppendVector(A=ScalarParameter() {defaultValue=1, parameterName=ScaleX}, B=ScalarParameter() {defaultValue=1, parameterName=ScaleY})
  n11 = Multiply(
   A=Constant3Vector() {constant=<rgba 1 0.0533 0 0>}
   B=TextureSample(UVs=Multiply(A=@n9, B=@n10)) {samplerType=SAMPLERTYPE_Masks, texture=&T_DeadDrop_LandingZone_M}.R
  )
  n12 = Multiply(A=@n11, B=Constant() {r=4})
  n13 = Constant()
  n14 = MaterialFunctionCall(Fade Length=Constant() {r=2048}, Fade Offset=Constant() {r=512}) {materialFunction=&CameraDepthFade}
  n15 = Constant() {r=768}
  n16 = Constant() {r=512}
  Multiply(
   A=LinearInterpolate(
    A=Add(A=@n8, B=@n12)
    B=LinearInterpolate(A=@n11, B=@n13, Alpha=@n14.Result)
    Alpha=MaterialFunctionCall(Fade Length=@n15, Fade Offset=@n16) {materialFunction=&CameraDepthFade}.Result
   )
   B=DepthFade(FadeDistance=Constant() {r=256})
  )
 ]
}
